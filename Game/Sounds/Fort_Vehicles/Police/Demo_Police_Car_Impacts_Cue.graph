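# Sound cue for police car impact sounds, in the impacts sound class with medium-object demolition-hit attenuation.
# A quality-level switch sends its first two slots to an evenly weighted random pick of three stereo impact waves through a modulator.
# Its last two slots play only the first wave, through a modulator with a changed maximum pitch.
# It references the player melee weapon concurrency group.
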SoundCue /Game/Sounds/Fort_Vehicles/Police/Demo_Police_Car_Impacts_Cue
root=QualityLevel_0
nodes [
 {id=Modulator_0, children [WavePlayer_0], properties {pitchMax=0.1}}
 {id=Modulator_1, children [Random_1]}
 {id=QualityLevel_0, children [Modulator_1,Modulator_1,Modulator_0,Modulator_0]}
 {id=Random_1, children [WavePlayer_3,WavePlayer_4,WavePlayer_5], properties {weights [1,1,1]}}
 {id=WavePlayer_0, properties {soundWaveAssetPtr=&/Police_Car_Impact_Stereo_01}}
 {id=WavePlayer_3, properties {soundWaveAssetPtr=&Police_Car_Impact_Stereo_01}}
 {id=WavePlayer_4, properties {soundWaveAssetPtr=&/Police_Car_Impact_Stereo_02}}
 {id=WavePlayer_5, properties {soundWaveAssetPtr=&/Police_Car_Impact_Stereo_03}}
]
properties {
 attenuationSettings=/Game/Sounds/Attenuation/Fort_Demolition_Hit_Medium_Sized_Object_Att
 maxDistance=4000
 soundClassObject=&Fort_Impacts
 volumeMultiplier=0.6
 concurrencyOverrides {maxCount=4, resolutionRule=StopOldest}
 concurrencySet [&Player_Melee_Weapon_Conc]
}
