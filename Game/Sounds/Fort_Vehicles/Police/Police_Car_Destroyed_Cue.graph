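# Sound cue for a destroyed police car, routed to the Fort_Impacts sound class.
# A single stereo wave player feeds a Modulator root, which varies pitch and volume at its default ranges.
# It uses a shared attenuation asset for demolition of medium-sized objects, a slight volume boost, and a concurrency override that prevents new instances once the limit is reached.
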SoundCue /Game/Sounds/Fort_Vehicles/Police/Police_Car_Destroyed_Cue
root=Modulator_6
nodes [
 {id=Modulator_6, children [WavePlayer_7]}
 {id=WavePlayer_7, properties {soundWaveAssetPtr=&/Police_Car_Destroyed_Stereo_01}}
]
properties {
 attenuationSettings=/Game/Sounds/Attenuation/Fort_Demolition_Destroy_Medium_Sized_Object_Att
 bOverrideConcurrency=true
 maxDistance=4000
 soundClassObject=&Fort_Impacts
 volumeMultiplier=1.2
 concurrencyOverrides {resolutionRule=PreventNew}
}
