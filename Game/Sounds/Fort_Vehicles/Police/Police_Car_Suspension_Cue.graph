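# Sound cue for police car suspension sounds, assigned to the Fort_Impacts sound class with boosted volume.
# A Random node picks one of three equally weighted stereo waves and passes it through a Modulator into a root Attenuation node with its own override.
# Cue-level attenuation overrides set HRTF spatialization with zero stereo spread, and a concurrency override caps playback at two instances, stopping the oldest.
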SoundCue /Game/Sounds/Fort_Vehicles/Police/Police_Car_Suspension_Cue
root=Attenuation_2
nodes [
 {id=Attenuation_2, children [Modulator_2], properties {bOverrideAttenuation=true, attenuationOverrides {nonSpatializedRadiusStart=200}}}
 {id=Modulator_2, children [Random_1]}
 {id=Random_1, children [WavePlayer_4,WavePlayer_5,WavePlayer_6], properties {bShouldExcludeFromBranchCulling=true, weights [1,1,1]}}
 {id=WavePlayer_4, properties {soundWaveAssetPtr=&/Police_Car_Suspension_Stereo_01}}
 {id=WavePlayer_5, properties {soundWaveAssetPtr=&/Police_Car_Suspension_Stereo_02}}
 {id=WavePlayer_6, properties {soundWaveAssetPtr=&/Police_Car_Suspension_Stereo_03}}
]
properties {
 attenuationSettings=/Game/Sounds/Attenuation/Fort_Prop_Vehicles_Att
 bOverrideAttenuation=true
 bOverrideConcurrency=true
 maxDistance=4000
 soundClassObject=&Fort_Impacts
 volumeMultiplier=2
 attenuationOverrides {spatializationAlgorithm=SPATIALIZATION_HRTF, stereoSpread=0}
 concurrencyOverrides {maxCount=2, resolutionRule=StopOldest}
}
