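# Sound cue named for a police car yelp siren, stored with the vehicle sounds, but with no nodes and no root, so it has no audio to play.
# Only playback settings remain: a reduced volume, the shared vehicle prop attenuation, a max distance and the impacts sound class.
# A concurrency override is set to prevent new instances once the limit is reached.
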SoundCue /Game/Sounds/Fort_Vehicles/Police/Police_Car_YelpSiren_Cue
root=null
nodes []
properties {
 attenuationSettings=/Game/Sounds/Attenuation/Fort_Prop_Vehicles_Att
 bOverrideConcurrency=true
 maxDistance=4000
 soundClassObject=&Fort_Impacts
 volumeMultiplier=0.4
 concurrencyOverrides {resolutionRule=PreventNew}
}
